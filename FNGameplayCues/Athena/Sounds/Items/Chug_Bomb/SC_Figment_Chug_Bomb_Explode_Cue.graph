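# Sound cue for the explosion of the Figment Chug Bomb item.
# A single explosion wave player feeds a Modulator that randomizes pitch on each play while holding volume steady.
# The cue plays through the large explosives sound class with shared smoke grenade attenuation and slightly reduced volume.
# Its own concurrency override prevents new instances once the limit is reached.
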SoundCue /FNGameplayCues/Athena/Sounds/Items/Chug_Bomb/SC_Figment_Chug_Bomb_Explode_Cue
root=Modulator_12
nodes [
 {id=Modulator_12, children [WavePlayer_0], properties {pitchMin=0.9, volumeMax=1, volumeMin=1}}
 {id=WavePlayer_0, properties {soundWaveAssetPtr=&/Chug_Bomb_Expl_01}}
]
properties {
 attenuationSettings=/Game/Sounds/Attenuation/Fort_Grenade_Smoke_Att
 bOverrideConcurrency=true
 maxDistance=6800
 soundClassObject=/Game/Sounds/FortSoundClassesAndModes/Fort_Explosives_Large_PSM
 volumeMultiplier=0.7
 attenuationOverrides {falloffDistance=2400, attenuationShapeExtents {x=600}}
 concurrencyOverrides {resolutionRule=PreventNew}
}
